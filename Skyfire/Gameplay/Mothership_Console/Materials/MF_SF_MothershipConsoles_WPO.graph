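# Material function that adds animated floating and wobbling World Position Offset to mothership console materials, passed through as material attributes.
# A sine drift between min and max vectors takes its phase from object position and its rate from a speed vector.
# It adds a sine-driven Z rotation around the object pivot, with vertex red varying phase and amount.
# Camera offsets feed both the rotation amount and the final offset.
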MaterialFunction /Skyfire/Gameplay/Mothership_Console/Materials/MF_SF_MothershipConsoles_WPO
graph {nodes [
  n1 = NamedRerouteUsage()
  n2 = MaterialFunctionCall(Color=Divide(A=ObjectPositionWS(), B=ScalarParameter() {defaultValue=128, group=WPO, parameterName=PositionOffset})) {materialFunction=&SplitComponents}
  n3 = VectorParameter() {group=WPO, parameterName=WPO_Speed, defaultValue=<rgba 0.1 0.05 0.075 2>}
  n4 = Add(A=Frac(None=Multiply(A=@n2.R, B=@n2.R)), B=Multiply(A=Time(), B=Multiply(A=@n3.RGB, B=@n3.RGB)))
  n5 = VertexColor()
  n6 = Sine(None=Add(A=Time(), B=LinearInterpolate(Alpha=@n5.R) {constA=-1, constB=2})) {period=5.8}
  n7 = Add(
   A=ScalarParameter() {defaultValue=0.0025, group=WPO, parameterName=Rotate_Amount}
   B=MaterialFunctionCall(Offset Amount=ScalarParameter() {defaultValue=0.001, group=WPO, parameterName=Rotate_CameraOffsetAmount}, World Position=ObjectPositionWS()) {materialFunction=/Engine/Functions/Engine_MaterialFunctions02/WorldPositionOffset/CameraOffset}.Clamped Camera Offset
  )
  n8 = Add(A=@n7, B=Multiply(A=@n7, B=LinearInterpolate(Alpha=@n5.R) {constA=0.3})) {constB=0.05}
  n9 = Add(
   A=LinearInterpolate(
    A=VectorParameter() {group=WPO, parameterName=WPO_Movement_Min, defaultValue=<rgba -1 1 0 1>}.RGB
    B=VectorParameter() {group=WPO, parameterName=WPO_Movement_Max, defaultValue=<rgba 1 1 4 1>}.RGB
    Alpha=Sine(None=@n4)
   )
   B=RotateAboutAxis(
    NormalizedRotationAxis=Constant3Vector() {constant=<rgba 0 0 1 0>}
    RotationAngle=Multiply(A=@n6, B=@n8) {constB=0.1}
    PivotPoint=ObjectPositionWS()
    Position=WorldPosition().XYZ
   )
  )
  FunctionOutput(None=SetMaterialAttributes!(MaterialAttributes=FunctionInput() {inputType=FunctionInput_MaterialAttributes}, WorldPositionOffset=Add(A=@n1, B=@n9)) {attributeSetTypes [#1]})
  NamedRerouteDeclaration(None=MaterialFunctionCall(Offset Amount=ScalarParameter() {defaultValue=1, group=WPO, parameterName=CameraOffsetAmount}, World Position=WorldPosition().XYZ) {materialFunction=&CameraOffset}.Camera Offset) {name=Cameraoffset, nodeColor=<rgba 0.125 0 1 1>}
 ]}
