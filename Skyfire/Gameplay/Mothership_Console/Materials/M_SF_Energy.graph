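# Skyfire energy material, apparently for the Mothership console meshes, given its location and preview mesh.
# MF_Skyfire_Energy builds the material attributes, which pass through MF_SF_MothershipConsoles_WPO, by its name adding console-specific world position offset, to the output.
# All shading logic lives in those two functions; the material itself is a Default Lit wrapper enabled for skeletal meshes and mesh deformers.
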Material /Skyfire/Gameplay/Mothership_Console/Materials/M_SF_Energy
graph {
 properties {
  bUseMaterialAttributes=true
  bUsedWithMeshDeformer=true
  bUsedWithSkeletalMesh=true
  pixelDepthOffsetMode=PDOM_Legacy
  previewMesh=/Skyfire/Gameplay/Mothership_Console/Meshes/S_Mothership_Consoles
  usedShadingModels=Default Lit
 }
 nodes [MaterialFunctionCall!(In=MaterialFunctionCall() {materialFunction=/Game/Environments/Apollo/Sets/Skyfire/Materials/MaterialFunctions/MF_Skyfire_Energy}.Result) {materialFunction=&/MF_SF_MothershipConsoles_WPO}]
}
