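# SkyFire back bling material function that takes Material Attributes and rewrites emissive and base color.
# Mask textures and a color tint drive the base color, while panned caustics, view-dot-normal inner glow and Fresnel edges build hue-shifted emissive layers.
# Pre-skinned local bounds derive a bottom mask and a jump-count color gradient.
# A time-based sine provides a switchable flash, and desaturated emissive weights the final base color blend.
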MaterialFunction /Skyfire/Cosmetics/Backbling/Materials/MF_SkyFire_BackBling
graph {nodes [
  n1 = NamedRerouteUsage()
  n2 = NamedRerouteUsage()
  n3 = Multiply(A=@n1, B=Multiply(A=NamedRerouteUsage(), B=@n2))
  Add(A=@n3, B=Multiply(A=@n1, B=@n2))
  n4 = GetMaterialAttributes(None=FunctionInput() {inputName=Material, inputType=FunctionInput_MaterialAttributes}) {attributeGetTypes [#1, #2, #3]}
  n5 = NamedRerouteUsage()
  n6 = TextureSampleParameter2D() {group=SkyFire, parameterName=SF_Mask_texture, texture=/Skyfire/Cosmetics/Backbling/Textures/T_SkyFire_Backpack_D}
  n7 = NamedRerouteDeclaration(None=Multiply(A=@n6.R) {constB=0.5}) {name=BaseColor_RedChannel, nodeColor=<rgba 0 0.9766 1 1>}
  n8 = Multiply(A=NamedRerouteUsage(), B=Multiply(A=@n7) {constB=0.2})
  n9 = NamedRerouteUsage()
  n10 = LinearInterpolate(A=Add(A=Multiply(A=@n5) {constB=0.5}, B=@n3), B=Add(A=NamedRerouteUsage(), B=Multiply(A=@n8, B=@n9)), Alpha=NamedRerouteUsage())
  n11 = NamedRerouteUsage()
  n12 = Add(
   A=NamedRerouteUsage()
   B=LinearInterpolate(
    A=@n10
    B=Multiply(A=@n10, B=@n11)
    Alpha=ScalarParameter() {group=SkyFire, parameterName="SF_Flash_Backpack 0=off 1=On", sliderMax=1, sortPriority=75}
   )
  )
  n13 = VectorParameter() {group=SkyFire, parameterName=SF_ColorTent, defaultValue=<rgba 0.1446 0.0899 0.4 1>}
  n14 = TextureSampleParameter2D() {group=SkyFire, parameterName=SF_Mask B_texture, texture=/Skyfire/Cosmetics/Backbling/Textures/T_SkyFireProto_Backpack_FX}
  n15 = Multiply(A=@n7, B=Multiply(A=OneMinus(None=@n14.G)) {constB=0.1})
  n16 = Constant3Vector()
  n17 = DotProduct(A=CameraVectorWS(), B=VertexNormalWS())
  n18 = TextureSampleParameter2D(UVs=Panner(Coordinate=@n17) {speedX=0.1, speedY=0.15}) {group=SkyFire, parameterName=SF_Distortion_texture, sortPriority=0, texture=/Game/Effects/Fort_Effects/Textures/Tile/T_Caustics_Chromatic}
  n19 = Power(Base=@n17, Exp=ScalarParameter() {defaultValue=15, group=SkyFire, parameterName=SF_EmissiveInnerGlowAmount}) {constExponent=5}
  n20 = NamedRerouteDeclaration(None=@n13.RGB) {name=Colortint, nodeColor=<rgba 0 1 0.5781 1>}
  n21 = NamedRerouteDeclaration(None=LinearInterpolate(A=@n20, B=NamedRerouteUsage())) {name=EmissiveHueshift, nodeColor=<rgba 0 0.5781 1 1>}
  n22 = Power(Base=@n17) {constExponent=50}
  n23 = Add(A=Multiply(A=Add(A=@n18.RGB, B=@n19), B=@n21), B=LinearInterpolate(A=@n22, B=Multiply(A=@n22, B=@n21)) {constAlpha=0.85})
  n24 = NamedRerouteUsage()
  n25 = Multiply(A=@n21, B=ScalarParameter() {defaultValue=2, group=SkyFire, parameterName=SF_EmissiveFresnelAmount}) {constB=2}
  n26 = Fresnel(
   ExponentIn=ScalarParameter() {defaultValue=6, group=SkyFire, parameterName=SF_EmissiveEdgeExponent}
   Normal=Transform(None=NamedRerouteUsage())
  ) {baseReflectFraction=0}
  n27 = Add(A=LinearInterpolate(A=Multiply(A=@n23) {constB=0.15}, B=@n23, Alpha=@n24), B=Multiply(A=Multiply(A=@n25, B=@n26) {constB=2}, B=@n24))
  n28 = NamedRerouteUsage()
  n29 = LinearInterpolate(
   A=LinearInterpolate(A=NamedRerouteUsage(), B=Multiply(A=@n13.RGB, B=@n15), Alpha=NamedRerouteUsage())
   B=@n16
   Alpha=Multiply(A=@n27, B=@n28)
  ) {constB=0}
  n30 = NamedRerouteUsage()
  n31 = Multiply(A=Desaturation(None=@n12)) {constB=4}
  FunctionOutput(None=SetMaterialAttributes!(MaterialAttributes=@n4.MaterialAttributes, EmissiveColor=@n12, BaseColor=LinearInterpolate(A=@n29, B=@n30, Alpha=@n31)) {attributeSetTypes [#1, #2]}) {outputName=Result_1}
  MaterialFunctionCall(Hue Shift Percentage=Constant() {r=0.05}, Texture=@n20) {materialFunction=&HueShift}
  Multiply(A=@n2, B=Multiply(A=NamedRerouteUsage()) {constB=4})
  NamedRerouteDeclaration(None=@n4.Normal) {name=ObjectNormal, nodeColor=<rgba 1 0.0469 0 1>}
  NamedRerouteDeclaration(None=Multiply(A=@n27, B=@n28)) {name=EmissiveEdge, nodeColor=<rgba 0 1 0.25 1>}
  NamedRerouteDeclaration(None=@n23) {name=Innerglow, nodeColor=<rgba 0.6172 1 0 1>}
  NamedRerouteDeclaration(None=Add(A=@n14.R, B=@n14.R)) {name=R+GChannel, nodeColor=<rgba 0 1 0.9531 1>}
  n32 = VectorParameter() {group=SkyFire, parameterName=SF_Jump_Color, defaultValue=<rgba 0.562 0.4262 1 2>}
  NamedRerouteDeclaration(None=@n32.RGB) {name=jumpcolor, nodeColor=<rgba 1 0 0.2812 1>}
  n33 = ScalarParameter() {defaultValue=0.025, desc="0=Darker 1=Brighter", group=SkyFire, parameterName=SF_BaseColor_DarkenMultiplier}
  n34 = PreSkinnedPosition()
  n35 = PreSkinnedLocalBounds()
  n36 = Divide(A=Subtract(A=VertexInterpolator(VS=@n34).PS, B=@n35.Min), B=Subtract(A=@n35.Max, B=@n35.Max))
  n37 = ComponentMask(None=@n36) {r=true}
  n38 = ScalarParameter() {defaultValue=2, group=SkyFire, parameterName=SF_Jump_Count, sortPriority=80}
  n39 = Divide(A=Add(A=Add(A=@n37) {constB=-1}, B=Divide(A=@n38))) {constB=0}
  NamedRerouteDeclaration(None=LinearInterpolate(A=Multiply(A=@n32.RGB, B=@n33) {constB=0.5}, B=@n32.RGB, Alpha=Saturate(None=@n39))) {name=Jump, nodeColor=<rgba 0 1 0.2734 1>}
  NamedRerouteDeclaration(None=@n16) {name=Black, nodeColor=<rgba 0 1 1 1>}
  n40 = Add(A=ComponentMask(None=@n36) {b=true}, B=Constant() {r=-0.65})
  NamedRerouteDeclaration(None=Saturate(None=Divide(A=@n40) {constB=0})) {name=BottomMask, nodeColor=<rgba 0 1 0.5313 1>}
  NamedRerouteDeclaration(None=@n4.BaseColor) {name=ObjectBaseColor, nodeColor=<rgba 1 0.1406 0 1>}
  NamedRerouteDeclaration(None=@n4.EmissiveColor) {name=ObjectEmissive, nodeColor=<rgba 0 1 0.2266 1>}
  NamedRerouteDeclaration(None=@n14.G) {name=GreenChannel, nodeColor=<rgba 0.7812 1 0 1>}
  NamedRerouteDeclaration(None=@n14.R) {name=RedChannel, nodeColor=<rgba 1 0.3281 0 1>}
  NamedRerouteDeclaration(None=@n13.RGB) {name=BaseColor, nodeColor=<rgba 1 0.4219 0 1>}
  n41 = Multiply(A=Time(), B=ScalarParameter() {defaultValue=1, group=SkyFire, parameterName=SF_Tic_Speed, sortPriority=90})
  n42 = Multiply(A=Add(A=Sine(None=@n41))) {constB=0.5}
  n43 = ScalarParameter() {defaultValue=0.1, group=SkyFire, parameterName=SF_Min_Flash_Brightness, sortPriority=85}
  NamedRerouteDeclaration(None=Saturate(None=Add(A=@n42, B=@n43))) {name=Flash, nodeColor=<rgba 0.7578 1 0 1>}
  Power(Base=@n6.R)
 ]}
