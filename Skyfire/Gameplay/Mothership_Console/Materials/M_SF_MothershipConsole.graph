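# Surface material for the Skyfire Mothership Console, built on material attributes and the shared character PreFX and PostFX functions.
# The PreFX result passes through the Skyfire back bling function, a weapon-skin thin-film layer and a console world position offset function before PostFX.
# A selectable channel of the mothership mask texture drives the thin-film alpha, with a scalar parameter blending between normal and inverted mask.
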Material /Skyfire/Gameplay/Mothership_Console/Materials/M_SF_MothershipConsole
graph {
 properties {
  bUseMaterialAttributes=true
  bUsedWithMeshDeformer=true
  bUsedWithSkeletalMesh=true
  pixelDepthOffsetMode=PDOM_Legacy
  refractionMethod=RM_IndexOfRefraction
  usedShadingModels=Default Lit
 }
 nodes [
  n1 = MaterialFunctionCall() {materialFunction=&BaseFN_CharacterFunction_PreFX}
  n2 = TextureSampleParameter2D() {group=ThinFilm, parameterName=Skin FX, sortPriority=-2, texture=/Skyfire/Gameplay/Mothership_Console/Textures/T_SF_Mothership_M}
  n3 = ChannelMaskParameter(None=@n2.RGBA) {group=ThinFilm, parameterName=ThinFilm Channel, sortPriority=-1}
  n4 = LinearInterpolate(
   A=@n3.R
   B=OneMinus(None=@n3.R)
   Alpha=ScalarParameter() {group=ThinFilm, parameterName="ThinFilm_InvertTexture 0=False 1=True", sortPriority=0}
  )
  n5 = MaterialFunctionCall!(Material=@n1.Result) {materialFunction=/Skyfire/Cosmetics/Backbling/Materials/MF_SkyFire_BackBling}
  MaterialFunctionCall(
   AO=@n1.AO
   Distance Blend=@n1.AO
   Fresnel=@n1.AO
   Gmap=@n1.AO
   Pre Skinned Local Position=@n1.AO
   Material=MaterialFunctionCall!(In=MaterialFunctionCall!(Alpha=@n4, Material Attributes=@n5.Result_1) {materialFunction=/Game/Weapons/WeaponSkins/MaterialFunction/ThinFilmFunction}.Result) {materialFunction=&/MF_SF_MothershipConsoles_WPO}.Result
  ) {materialFunction=&BaseFN_CharacterFunction_PostFX}
  MaterialFunctionCall(
   PreskinnedPosition=@n1.Pre skinned Local Position
   Gmap=@n1.Pre skinned Local Position
   MaterialAttributes=@n1.Pre skinned Local Position
   AO=@n1.Pre skinned Local Position
  ) {materialFunction=&CharacterSkinFX}
 ]
}
