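# Material function for Skyfire mothership consoles that layers masked, colored glow onto an incoming material.
# A console mask shapes the emissive color, with an optional global glow boost; base color darkens where the glow is bright.
# It also builds a parallax logo with rotating backdrop, triplanar tech pattern, Fresnel edge glow, screen-space scanlines and noise-based glitch.
# Named reroutes connect these parts, though the listing does not show which reroute each usage reads.
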MaterialFunction /Skyfire/Gameplay/Mothership_Console/Materials/MF_SF_MothershipConsoles
graph {nodes [
  n1 = Add(A=TextureCoordinate(), B=ComponentMask(None=VectorParameter() {group=Skyfire_Logo, parameterName=SF_Logo_Offset}.RGB) {g=true, r=true})
  n2 = MaterialFunctionCall(
   UVs=BumpOffset(Coordinate=@n1, Height=Constant() {r=1}, HeightRatioInput=Constant() {r=0.175})
   Texture Scale=ComponentMask(None=VectorParameter() {group=Skyfire_Logo, parameterName=SF_Logo_Scale, defaultValue=<rgba 0.65 0.65 0 0>}.RGB) {g=true, r=true}
  ) {materialFunction=&ScaleUVsByCenter}
  n3 = TextureSample(UVs=@n2.UVs, Tex=NamedRerouteUsage()) {samplerSource=SSM_Clamp_WorldGroupSettings, texture=&T_BlacksRGB}
  n4 = VectorParameter() {group=Skyfire_Logo, parameterName=SF_LogoBG_Scale, defaultValue=<rgba 0.75 0.75 0 0>}
  n5 = MaterialFunctionCall(
   UVs=MaterialFunctionCall(UVs=@n1, Texture Scale=ComponentMask(None=@n4.RGB) {g=true, r=true}) {materialFunction=&ScaleUVsByCenter}.UVs
   Rotation Center=Constant2Vector() {g=0.5, r=0.5}
   "Rotation Angle (0-1)"=Multiply(A=Time()) {constB=0.025}
  ) {materialFunction=&CustomRotator}
  n6 = NamedRerouteUsage()
  Add(
   A=@n3.R
   B=Multiply(A=TextureSample(UVs=@n5.Rotated Values, Tex=@n6) {samplerSource=SSM_Clamp_WorldGroupSettings, texture=&T_BlacksRGB}.G) {constB=0.25}
  )
  n7 = Multiply(A=Time(), B=ScalarParameter() {defaultValue=1, group=Skyfire_Scanline, parameterName=Kiwi_Scanline_Speed})
  n8 = Fresnel() {baseReflectFraction=0, exponent=3}
  n9 = VertexNormalWS()
  n10 = CameraVectorWS()
  n11 = Multiply(
   A=DotProduct(A=ComponentMask(None=@n9) {b=true}, B=ComponentMask(None=@n10) {b=true})
   B=ScalarParameter() {defaultValue=0.5, group=Skyfire_Scanline, parameterName=Kiwi_Scanline_LineBend}
  )
  n12 = Constant()
  n13 = Add(
   A=QualitySwitch(Default=Multiply(A=@n8, B=@n11) {constB=0.2}, Low=@n12, Medium=@n12)
   B=ComponentMask(None=ScreenPosition().ViewportUV) {g=true}
  )
  n14 = Multiply(A=@n13, B=ScalarParameter() {defaultValue=100, group=Skyfire_Scanline, parameterName=Kiwi_Scanline_Tile})
  n15 = ConstantBiasScale(None=Sine(None=Add(A=@n7, B=@n14)))
  n16 = ScalarParameter() {defaultValue=1500, group=Skyfire_Scanline, parameterName=Kiwi_Scanline_FadeDistance}
  n17 = Multiply(
   A=LinearInterpolate(
    A=Power(Base=@n15)
    Alpha=MaterialFunctionCall(Distance=@n16) {materialFunction=&Near_Camera_Fade}.Result
   ) {constA=0.4, constB=0.3}
   B=ScalarParameter() {defaultValue=0.05, group=Skyfire_Scanline, parameterName=Kiwi_Scanline_DarkenBrights}
  ) {constB=0.25}
  Add(A=@n17, B=ScalarParameter() {group=Skyfire_Scanline, parameterName=Kiwi_Scanline_BrightenDarks})
  n18 = GetMaterialAttributes(None=FunctionInput() {inputName=Material, inputType=FunctionInput_MaterialAttributes}) {attributeGetTypes [#1, #2]}
  n19 = NamedRerouteUsage()
  n20 = NamedRerouteUsage()
  n21 = TextureSampleParameter2D() {
   constCoordinate=1
   group=SkyFire
   parameterName=SF_Mask_texture
   sortPriority=0
   texture=/Skyfire/Gameplay/Mothership_Console/Textures/T_SF_Mothership_M
  }
  n22 = Multiply(A=Add(A=@n21.R) {constB=0.02}) {constB=0.33}
  n23 = Add(A=@n19, B=Multiply(A=@n20, B=@n22))
  n24 = Add(A=NamedRerouteUsage(), B=NamedRerouteUsage())
  n25 = NamedRerouteUsage()
  n26 = NamedRerouteUsage()
  n27 = Multiply(A=Power(Base=@n22) {constExponent=1.85}) {constB=25}
  n28 = Add(A=@n19, B=LinearInterpolate(A=Multiply(A=@n24, B=@n25), B=Multiply(A=@n26, B=@n27), Alpha=@n27))
  n29 = NamedRerouteUsage()
  n30 = NamedRerouteUsage()
  n31 = Add(
   A=LinearInterpolate(A=LinearInterpolate(A=@n23, B=@n28, Alpha=@n27), B=NamedRerouteUsage(), Alpha=Multiply(A=@n29) {constB=0.2})
   B=Multiply(A=@n30, B=NamedRerouteUsage())
  )
  n32 = LinearInterpolate(
   A=@n31
   B=Multiply(A=@n31, B=ScalarParameter() {defaultValue=2, group=SkyFire, parameterName=SF_GlobalGlowAmount, sortPriority=-9}) {constB=2}
   Alpha=ScalarParameter() {group=SkyFire, parameterName=SF_GlobalGlowOn, sortPriority=-10}
  )
  n33 = NamedRerouteUsage()
  n34 = Constant3Vector()
  n35 = Multiply(A=Desaturation(None=@n32)) {constB=4}
  FunctionOutput(None=SetMaterialAttributes!(
    MaterialAttributes=@n18.MaterialAttributes
    EmissiveColor=Add(A=@n18.EmissiveColor, B=@n32)
    BaseColor=LinearInterpolate(A=@n33, B=@n34, Alpha=@n35)
   ) {attributeSetTypes [#1, #3]}) {outputName=Result_1}
  n36 = TextureCoordinate()
  n37 = VectorParameter() {group=Skyfire_Timer, parameterName=SF_Font_Offset}
  n38 = MaterialFunctionCall(
   UVs=Add(A=NamedRerouteDeclaration(None=@n36) {name=TexUVs, nodeColor=<rgba 1 0 0.375 1>}, B=ComponentMask(None=@n37.RGB) {g=true, r=true})
   Texture Scale=ComponentMask(None=VectorParameter() {group=Skyfire_Timer, parameterName=SF_Font_Size, defaultValue=<rgba 0.6 0.3 0 1>}.RGB) {g=true, r=true}
  ) {materialFunction=&ScaleUVsByCenter}
  NamedRerouteDeclaration(None=@n38.UVs) {name=ScaleByCenterUVs, nodeColor=<rgba 0 0.7891 1 1>}
  NamedRerouteDeclaration(None=MaterialFunctionCall!(UseLocalSpace, WorldPosition=WorldPosition().XYZ, NormalOverride) {materialFunction=/Game/Environments/Apollo/Sets/Skyfire/Materials/MaterialFunctions/MF_WorldAlignedUVs_HardEdge}.Result) {name=TriplanarUVs, nodeColor=<rgba 0 0.5547 1 1>}
  NamedRerouteDeclaration(None=@n3.R) {name=Logo_HashMarks, nodeColor=<rgba 1 0 0.0703 1>}
  NamedRerouteDeclaration(None=@n3.A) {name=Logo_Spaceship, nodeColor=<rgba 0 0.6016 1 1>}
  n39 = TextureObjectParameter() {group=Skyfire_Textures, parameterName=SF_Logo, sortPriority=-100, texture=/Skyfire/Gameplay/Mothership_Console/Textures/T_SF_Logo2_M}
  NamedRerouteDeclaration(None=@n39) {name=Glyph01_Masks, nodeColor=<rgba 0.5703 1 0 1>}
  NamedRerouteDeclaration(None=@n38.0-1 mask) {name=ScaleByCenterUVMask, nodeColor=<rgba 0 0.5547 1 1>}
  n40 = Add(A=ScalarParameter() {group=SkyFire, parameterName=Kiwi_GradientOffset}, B=ComponentMask(None=TextureCoordinate()) {g=true})
  n41 = ScalarParameter() {defaultValue=8, group=SkyFire, parameterName=Kiwi_GradientFalloff}
  NamedRerouteDeclaration(None=Multiply(A=Divide(A=@n40, B=@n41), B=ScalarParameter() {defaultValue=4, group=SkyFire, parameterName=Kiwi_GradientIntensity})) {name=Gradient, nodeColor=<rgba 1 0 0.8437 1>}
  NamedRerouteDeclaration(None=VectorParameter() {group=SkyFire, parameterName=SF_MainColor, defaultValue=<rgba 0.0689 0.0034 0.2 1>}.RGB) {name=ColorTentMPC, nodeColor=<rgba 0 0.0859 1 1>}
  NamedRerouteDeclaration(None=@n21.R) {name=FXMask, nodeColor=<rgba 1 0.7031 0 1>}
  NamedRerouteDeclaration(None=@n2.UVs) {name=Glyph UVs, nodeColor=<rgba 1 0 0.0469 1>}
  n42 = ScalarParameter() {defaultValue=100, group=TechPattern, parameterName=SF_TechPatternTileDivisorU}
  n43 = ScalarParameter() {defaultValue=100, group=TechPattern, parameterName=SF_TechPatternTileDivisorV}
  NamedRerouteDeclaration(None=Divide(A=NamedRerouteUsage(), B=AppendVector(A=@n42, B=@n43))) {name=TechPatternUVs, nodeColor=<rgba 0 1 0.7188 1>}
  NamedRerouteDeclaration(None=@n17) {name=Scanlines, nodeColor=<rgba 1 0.3984 0 1>}
  NamedRerouteDeclaration(None=@n30) {name=LogoColor, nodeColor=<rgba 0.2188 0 1 1>}
  n44 = NamedRerouteUsage()
  n45 = Multiply(A=Power(Base=@n22) {constExponent=1.25}) {constB=10}
  NamedRerouteDeclaration(None=Add(A=Multiply(A=@n44, B=@n20), B=Multiply(A=@n20, B=@n45))) {name=BaseColor, nodeColor=<rgba 0 0.7188 1 1>}
  n46 = TextureSampleParameter2D(UVs=NamedRerouteUsage()) {
   group=TechPattern
   parameterName=SF_PackA
   samplerSource=SSM_Wrap_WorldGroupSettings
   samplerType=SAMPLERTYPE_LinearColor
   texture=/Game/Environments/Apollo/Sets/Skyfire/Textures/T_Skyfire_PackA
  }
  NamedRerouteDeclaration(None=@n46.G) {name=Height * WorldMask, nodeColor=<rgba 0 1 0.3906 1>}
  NamedRerouteDeclaration(None=@n46.G) {name=HeightMask, nodeColor=<rgba 0 1 0.5312 1>}
  NamedRerouteDeclaration(None=LinearInterpolate(A=NamedRerouteDeclaration(None=@n20) {name=HueshiftedColor, nodeColor=<rgba 1 0 0.9844 1>}, B=NamedRerouteUsage())) {name=Logo+BaseColor, nodeColor=<rgba 0 0.4375 1 1>}
  n47 = NamedRerouteUsage()
  n48 = VectorParameter() {group=Skyfire_Glitch, parameterName=Kiwi_GlitchPanner01, defaultValue=<rgba 0 0.5 0 0.2>}
  n49 = Multiply(
   A=Divide(A=@n47, B=ScalarParameter() {defaultValue=25, group=Skyfire_Glitch, parameterName=Kiwi_Glitch_Tilerate1})
   B=AppendVector(A=@n48.R, B=@n48.R)
  )
  n50 = AppendVector(A=@n48.B, B=@n48.B)
  n51 = TextureSample(UVs=Add(A=Panner(Coordinate=@n49, Speed=@n50) {bFractionalPart=true})) {samplerSource=SSM_Wrap_WorldGroupSettings, samplerType=SAMPLERTYPE_Normal, texture=/Game/Global/Textures/Default/Noise/64x64Noise}
  n52 = VectorParameter() {group=Skyfire_Glitch, parameterName=Kiwi_Glitch_Panner02, defaultValue=<rgba 0 0.5 0 -0.1>}
  n53 = Multiply(
   A=Divide(A=@n47, B=ScalarParameter() {defaultValue=25, group=Skyfire_Glitch, parameterName=Kiwi_Glitch_Tilerate})
   B=AppendVector(A=@n52.R, B=@n52.R)
  )
  n54 = AppendVector(A=@n52.B, B=@n52.B)
  n55 = TextureSample(UVs=Add(A=Panner(Coordinate=@n53, Speed=@n54) {bFractionalPart=true})) {samplerSource=SSM_Wrap_WorldGroupSettings, samplerType=SAMPLERTYPE_Normal, texture=&64x64Noise}
  n56 = Divide(
   A=@n47
   B=AppendVector(
    A=ScalarParameter() {defaultValue=100, group=Skyfire_Glitch, parameterName=Kiwi_Glitch_NoiseTilerateX}
    B=ScalarParameter() {defaultValue=250, group=Skyfire_Glitch, parameterName=Kiwi_Glitch_NoiseTilerateY}
   )
  )
  n57 = VectorParameter() {group=Skyfire_Glitch, parameterName=Kiwi_Glitch_NoisePanandSpeed, defaultValue=<rgba 1 1 5 0.15>}
  n58 = AppendVector(A=@n57.R, B=@n57.R)
  n59 = TextureSample(UVs=Panner(Coordinate=Multiply(A=@n56, B=@n58), Speed=AppendVector(A=@n57.B, B=@n57.B)) {bFractionalPart=true}) {
   samplerSource=SSM_Wrap_WorldGroupSettings
   samplerType=SAMPLERTYPE_LinearColor
   texture=/Game/Global/Textures/Default/Noise/T_Wrap_SnowStreak_Noise
  }
  n60 = Desaturation(None=Multiply(A=Multiply(A=@n51.R, B=@n55.G), B=@n59.G))
  n61 = Saturate(None=Round(None=Multiply(A=@n60) {constB=10}))
  n62 = Add(
   A=Multiply(A=@n60, B=ScalarParameter() {defaultValue=4, group=Skyfire_Glitch, parameterName=Kiwi_Glitch_DarkenBrights}) {constB=0.25}
   B=ScalarParameter() {group=Skyfire_Glitch, parameterName=Kiwi_Glitch_BrightenDarks}
  )
  NamedRerouteDeclaration(None=Add(A=NamedRerouteDeclaration(None=@n61) {name=scanlineRounded, nodeColor=<rgba 1 0.7031 0 1>}, B=Saturate(None=@n62))) {name=Glitch Transition, nodeColor=<rgba 0.1719 1 0 1>}
  n63 = TextureSampleParameter2D(UVs=NamedRerouteUsage()) {
   group=TechPattern
   parameterName=SF_PackA
   samplerSource=SSM_Wrap_WorldGroupSettings
   samplerType=SAMPLERTYPE_LinearColor
   texture=&T_Skyfire_PackA
  }
  NamedRerouteDeclaration(None=LinearInterpolate(A=@n63.B, Alpha=ScalarParameter() {group=TechPattern, parameterName=SF_Blocks_AO}) {constA=1}) {name=BlocksAO, nodeColor=<rgba 0.8984 0 1 1>}
  NamedRerouteDeclaration(None=@n13) {name=ScreenspaceUVs, nodeColor=<rgba 0 0.8359 1 1>}
  n64 = Time()
  n65 = ScalarParameter() {defaultValue=10, group=TechPattern, parameterName=SF_BlocksEmissiveTimeScale}
  n66 = Sine(None=Add(A=Divide(A=@n64, B=@n65), B=NamedRerouteUsage())) {period=0.15}
  n67 = NamedRerouteUsage()
  NamedRerouteDeclaration(None=Multiply(A=ConstantBiasScale(None=@n66), B=OneMinus(None=@n67))) {name=TechPattern, nodeColor=<rgba 1 0.0938 0 1>}
  n68 = Multiply(A=NamedRerouteUsage(), B=Multiply(A=NamedRerouteUsage()) {constB=0.2})
  n69 = ScalarParameter() {group=TechPattern, parameterName=SF_TechPatternOn}
  n70 = DotProduct(A=CameraVectorWS(), B=VertexNormalWS())
  n71 = TextureSampleParameter2D(UVs=Panner(Coordinate=@n70) {speedX=0.025, speedY=0.05}) {group=SkyFire, parameterName=SF_Distortion_texture, sortPriority=0, texture=/Game/Effects/Fort_Effects/Textures/Tile/T_Caustics_Chromatic}
  n72 = ScalarParameter() {defaultValue=20, group=SkyFire, parameterName=SF_EmissiveInnerGlowAmount}
  n73 = Add(A=Multiply(A=Desaturation(None=@n71.RGB)) {constB=0.25}, B=Multiply(A=Power(Base=@n70, Exp=@n72) {constExponent=5}) {constB=0.5})
  n74 = NamedRerouteUsage()
  n75 = Multiply(A=Power(Base=@n70) {constExponent=100}) {constB=0.25}
  n76 = NamedRerouteUsage()
  n77 = Multiply(A=Add(A=Multiply(A=@n73, B=@n74), B=Multiply(A=@n75, B=@n76))) {constB=0.15}
  n78 = Multiply(
   A=Multiply(A=NamedRerouteUsage()) {constB=1.5}
   B=Fresnel(ExponentIn=ScalarParameter() {defaultValue=4, group=SkyFire, parameterName=SF_EmissiveEdgeExponent}) {baseReflectFraction=0}
  ) {constB=2}
  NamedRerouteDeclaration(None=Add(A=Multiply(A=@n68, B=@n69), B=Add(A=@n77, B=@n78))) {name=EmissiveEdge, nodeColor=<rgba 0 1 0.25 1>}
  NamedRerouteDeclaration(None=VectorParameter() {parameterName=SF_LogoColor, defaultValue=<rgba 0.3908 0.0897 0.5 1>}.RGB) {name=FontColorParamMPC, nodeColor=<rgba 0.4297 0 1 1>}
  NamedRerouteDeclaration(None=@n18.Normal) {name=ObjectNormals, nodeColor=<rgba 0.125 0 1 1>}
  ScalarParameter() {defaultValue=1, group=Skyfire_Logo, parameterName=SF_LogoBG_Scale}
  TextureCoordinate()
 ]}
