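# Screen material for the Skyfire Mothership Console, usable on skeletal meshes and mesh deformers.
# It builds default-lit material attributes from a black constant for base color and emissive, plus Kiwi-prefixed metallic, specular and roughness parameters.
# The attributes pass through the shared MF_SF_MothershipConsoles function, then MF_SF_MothershipConsoles_WPO, whose name indicates world position offset.
# The graph itself has no screen textures or animation; any such look comes from those functions.
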Material /Skyfire/Gameplay/Mothership_Console/Materials/M_SF_MothershipConsole_Screen
graph {
 properties {
  bUseMaterialAttributes=true
  bUsedWithMeshDeformer=true
  bUsedWithSkeletalMesh=true
  pixelDepthOffsetMode=PDOM_Legacy
  previewMesh=/Skyfire/Gameplay/Mothership_Console/Meshes/S_Mothership_Consoles
  usedShadingModels=Default Lit
 }
 nodes [
  n1 = Constant3Vector()
  n2 = SetMaterialAttributes!(
   MaterialAttributes
   Base Color=@n1
   Metallic=ScalarParameter() {parameterName=Kiwi_Metallic}
   Specular=ScalarParameter() {defaultValue=0.5, parameterName=Kiwi_Specular}
   Roughness=ScalarParameter() {defaultValue=0.6, parameterName=Kiwi_Roughness}
   Emissive Color=@n1
  ) {attributeSetTypes [
    #1
    #2
    #3
    #4
    #5
   ]}
  MaterialFunctionCall!(In=MaterialFunctionCall!(Material=@n2) {materialFunction=&/MF_SF_MothershipConsoles}.Result_1) {materialFunction=&/MF_SF_MothershipConsoles_WPO}
 ]
}
